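# Sound cue for the team score chest display on the reward screen, judging by its name.
# A single wave player feeds the output directly with the matching chest display sound wave, with no randomization, mixing or modulation.
# The cue is assigned to the Fort_UI sound class.
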SoundCue /Game/Sounds/Fort_UI_Sounds/Rewards/Fort_RewardScreen_TeamScore_ChestDisplay_Cue
root=WavePlayer_0
nodes [{id=WavePlayer_0, properties {soundWaveAssetPtr=&/Fort_RewardScreen_TeamScore_ChestDisplay}}]
properties {maxDistance=2097152, soundClassObject=&Fort_UI, volumeMultiplier=1}
